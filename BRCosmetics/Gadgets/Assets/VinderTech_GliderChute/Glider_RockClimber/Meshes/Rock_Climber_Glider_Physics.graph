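# Physics asset for the Rock Climber glider cosmetic, giving its skeletal mesh one capsule body per simulated or collidable bone.
# Constraints form three unlinked clusters: head, wings and rear tail on Tail_01; handles, shelf and ropes branching from Handles_Base; four stand legs on Stand_up.
# Every joint shares the same 45-degree swing and twist limit values, and only one rope on the left and one in the middle have bodies.
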PhysicsAsset /BRCosmetics/Gadgets/Assets/VinderTech_GliderChute/Glider_RockClimber/Meshes/Rock_Climber_Glider_Physics
bodies {
 Handle_01_l [{shapeType=Capsule, length=21.4028, radius=18.2432, center=<-20.3697 9.2827 5.3728>, rotation=<rot 34.2728 62.4126 -90>}]
 Handle_01_r [{shapeType=Capsule, length=21.4028, radius=18.2432, center=<20.3697 -9.2828 -5.3728>, rotation=<rot 34.2725 62.4128 -90>}]
 Handle_Shelf [{shapeType=Capsule, length=90.9906, radius=27.5424, center=<-5.3815 -1.1225 2.5888>, rotation=<rot 0 90.7823 -89.2435>}]
 Handles_Base [{shapeType=Capsule, length=89.082, radius=56.3629, center=<-0.0657 -2.3944 3.1318>, rotation=<rot 0 -90.3201 -89.9982>}]
 Head [{shapeType=Capsule, length=60.0607, radius=52.9961, center=<1.2851 0.4481 0.4349>, rotation=<rot 6.1088 -178.8587 -90>}]
 Rope_left_01 [{shapeType=Capsule, length=17.5894, radius=13.8466, center=<24.7117 1.0399 1.3585>, rotation=<rot 0.0021 -90 0.1698>}]
 Ropes_Base [{shapeType=Capsule, length=19.5748, radius=15.6429, center=<16.2153 -0.4767 0.6142>, rotation=<rot 0 -94.2747 -87.8007>}]
 Ropes_Mid_01 [{shapeType=Capsule, length=16.275, radius=12.4916, center=<21.8198 -0.2697 0.782>, rotation=<rot -1.3757 -90 -0.8751>}]
 Stand_bk_up_l [{shapeType=Capsule, length=28.8666, radius=3.8926, center=<-32.7931 0.0001 -0.0457>, rotation=<rot 0 -90.0001 -89.625>}]
 Stand_bk_up_r [{shapeType=Capsule, length=28.8666, radius=3.8926, center=<32.7927 0 0.0455>, rotation=<rot 0 -90.0001 -89.6249>}]
 Stand_ft_up_l [{shapeType=Capsule, length=28.8666, radius=3.8925, center=<-32.8072 -0.0001 -0.0457>, rotation=<rot 0 -89.9999 -89.6253>}]
 Stand_ft_up_r [{shapeType=Capsule, length=28.8666, radius=3.8925, center=<32.8068 0 0.0454>, rotation=<rot 0 -90 -89.6252>}]
 Stand_up [{shapeType=Capsule, length=26.1112, radius=24.1821, center=<0 -0.0003 -0.2383>, rotation=<rot 0 180 -89.9923>}]
 Tail_01 [{shapeType=Capsule, length=77.5447, radius=48.6116, center=<49.3245 -0.0002 -9.4399>, rotation=<rot 5.8252 -179.9997 -90>}]
 Tail_02 [{shapeType=Capsule, length=72.8954, radius=17.4884, center=<62.5526 -0.0058 -1.8983>, rotation=<rot 0 -90.005 -88.2661>}]
 Wing_Base_l [{shapeType=Capsule, length=146.5445, radius=78.4168, center=<-135.2588 -6.2852 39.8316>, rotation=<rot 0 -101.0502 -71.3513>}]
 Wing_Base_r [{shapeType=Capsule, length=146.553, radius=78.4181, center=<135.2577 6.2429 -39.837>, rotation=<rot 0 -101.0737 -71.3523>}]
}
constraints [
 {bone1Name=Head, bone2Name=Tail_01, swing1LimitDegrees=45, swing2LimitDegrees=45, twistLimitDegrees=45}
 {bone1Name=Wing_Base_l, bone2Name=Tail_01, swing1LimitDegrees=45, swing2LimitDegrees=45, twistLimitDegrees=45}
 {bone1Name=Wing_Base_r, bone2Name=Tail_01, swing1LimitDegrees=45, swing2LimitDegrees=45, twistLimitDegrees=45}
 {bone1Name=Tail_02, bone2Name=Tail_01, swing1LimitDegrees=45, swing2LimitDegrees=45, twistLimitDegrees=45}
 {bone1Name=Handle_Shelf, bone2Name=Handles_Base, swing1LimitDegrees=45, swing2LimitDegrees=45, twistLimitDegrees=45}
 {bone1Name=Ropes_Base, bone2Name=Handle_Shelf, swing1LimitDegrees=45, swing2LimitDegrees=45, twistLimitDegrees=45}
 {bone1Name=Rope_left_01, bone2Name=Ropes_Base, swing1LimitDegrees=45, swing2LimitDegrees=45, twistLimitDegrees=45}
 {bone1Name=Ropes_Mid_01, bone2Name=Ropes_Base, swing1LimitDegrees=45, swing2LimitDegrees=45, twistLimitDegrees=45}
 {bone1Name=Handle_01_l, bone2Name=Handles_Base, swing1LimitDegrees=45, swing2LimitDegrees=45, twistLimitDegrees=45}
 {bone1Name=Handle_01_r, bone2Name=Handles_Base, swing1LimitDegrees=45, swing2LimitDegrees=45, twistLimitDegrees=45}
 {bone1Name=Stand_ft_up_r, bone2Name=Stand_up, swing1LimitDegrees=45, swing2LimitDegrees=45, twistLimitDegrees=45}
 {bone1Name=Stand_bk_up_r, bone2Name=Stand_up, swing1LimitDegrees=45, swing2LimitDegrees=45, twistLimitDegrees=45}
 {bone1Name=Stand_ft_up_l, bone2Name=Stand_up, swing1LimitDegrees=45, swing2LimitDegrees=45, twistLimitDegrees=45}
 {bone1Name=Stand_bk_up_l, bone2Name=Stand_up, swing1LimitDegrees=45, swing2LimitDegrees=45, twistLimitDegrees=45}
]
